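# Sound cue for the completion of a Creative timed objective toy.
# A mixer layers three one-shots: a low timer beep, a slightly delayed completion sound at the highest level, and a quiet scoring sound.
# It uses small-toy attenuation and the quests sound class, with a concurrency override allowing one instance per owner that stops the oldest.
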SoundCue /Game/Sounds/Creative/Toys/TimedObjective/TimedObjective_Completion_Cue
root=Mixer_0
nodes [
 {id=Delay_4, children [WavePlayer_12], properties {delayMax=0.02, delayMin=0.02}}
 {id=Mixer_0, children [WavePlayer_8,Delay_4,WavePlayer_2], properties {inputVolume [0.7,1,0.2]}}
 {id=WavePlayer_12, properties {soundWaveAssetPtr=&/TimedObjective_Completion_01}}
 {id=WavePlayer_2, properties {soundWaveAssetPtr=/Game/Sounds/Creative/Modes/Scoring/Score_Enemy_Up_01}}
 {id=WavePlayer_8, properties {soundWaveAssetPtr=/Game/Sounds/Fort_GamePlay_Sounds/Explosives/timer_beep_lo3}}
]
properties {
 attenuationSettings=/Game/Sounds/Creative/Attenuations/Fort_Creative_SmallToy_Att
 bOverrideConcurrency=true
 maxDistance=3000
 soundClassObject=&Fort_Quests
 volumeMultiplier=1.344
 concurrencyOverrides {bLimitToOwner=true, maxCount=1, resolutionRule=StopOldest}
}
